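# Looping ambient Sound Cue for a flock of bats, mixing a chirp layer and a wing flutter layer.
# Each layer has its own attenuation and loops a random delay into pitch and volume modulation of a randomly picked one-shot.
# The chirp pool adds a silence clip and two empty slots for sparser calls.
# Per-owner concurrency caps voices, and sends route to exterior ambient bus and reverb.
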SoundCue /Game/Sounds/Asteria/Environment_Actors/Bats/Spot_Bats_Flocking_Loop_Cue
root=Mixer_0
nodes [
 {
  id=Attenuation_0
  children [Looping_1]
  properties {attenuationSettings=&/Fort_Prop_Bats_Chirps_Att}
 }
 {
  id=Attenuation_1
  children [Looping_3]
  properties {attenuationSettings=&/Fort_Prop_Bats_Wings_Att}
 }
 {id=Delay_0, children [Modulator_3], properties {delayMax=0.8, delayMin=0.1}}
 {id=Delay_1, children [Modulator_2], properties {delayMax=2, delayMin=0.75}}
 {id=Looping_1, children [Delay_1]}
 {id=Looping_3, children [Delay_0]}
 {id=Mixer_0, children [Attenuation_0, Attenuation_1], properties {inputVolume [1,1]}}
 {id=Modulator_2, children [Random_2], properties {pitchMax=1.2, pitchMin=0.8, volumeMax=0.9, volumeMin=0.6}}
 {id=Modulator_3, children [Random_1], properties {pitchMax=1.5, pitchMin=0.9, volumeMax=1.1, volumeMin=0.9}}
 {
  id=Random_1
  children [WavePlayer_3, WavePlayer_4, WavePlayer_5, WavePlayer_6, WavePlayer_7, WavePlayer_8, WavePlayer_9, WavePlayer_10]
  properties {weights [1,1,1,1,1,1,1,1]}
 }
 {
  id=Random_2
  children [WavePlayer_11, WavePlayer_12, WavePlayer_13, WavePlayer_14, WavePlayer_15, WavePlayer_16, WavePlayer_0, null, null]
  properties {bShouldExcludeFromBranchCulling=true, weights [1,1,1,1,1,1,1,1,1]}
 }
 {id=WavePlayer_0, properties {soundWaveAssetPtr=/Game/Sounds/Fort_Test_Sounds/5_seconds_silence}}
 {id=WavePlayer_10, properties {soundWaveAssetPtr=&/Spot_Bat_Wing_Flutter_Oneshot_08}}
 {id=WavePlayer_11, properties {soundWaveAssetPtr=&/Spot_Bat_Chirp_01}}
 {id=WavePlayer_12, properties {soundWaveAssetPtr=&/Spot_Bat_Chirp_02}}
 {id=WavePlayer_13, properties {soundWaveAssetPtr=&/Spot_Bat_Chirp_03}}
 {id=WavePlayer_14, properties {soundWaveAssetPtr=&/Spot_Bat_Chirp_04}}
 {id=WavePlayer_15, properties {soundWaveAssetPtr=&/Spot_Bat_Chirp_05}}
 {id=WavePlayer_16, properties {soundWaveAssetPtr=&/Spot_Bat_Chirp_06}}
 {id=WavePlayer_3, properties {soundWaveAssetPtr=&/Spot_Bat_Wing_Flutter_Oneshot_01}}
 {id=WavePlayer_4, properties {soundWaveAssetPtr=&/Spot_Bat_Wing_Flutter_Oneshot_02}}
 {id=WavePlayer_5, properties {soundWaveAssetPtr=&/Spot_Bat_Wing_Flutter_Oneshot_03}}
 {id=WavePlayer_6, properties {soundWaveAssetPtr=&/Spot_Bat_Wing_Flutter_Oneshot_04}}
 {id=WavePlayer_7, properties {soundWaveAssetPtr=&/Spot_Bat_Wing_Flutter_Oneshot_05}}
 {id=WavePlayer_8, properties {soundWaveAssetPtr=&/Spot_Bat_Wing_Flutter_Oneshot_06}}
 {id=WavePlayer_9, properties {soundWaveAssetPtr=&/Spot_Bat_Wing_Flutter_Oneshot_07}}
]
properties {
 attenuationSettings=&/Fort_Prop_Bats_Flocking_Att
 bOverrideConcurrency=true
 maxDistance=4000
 soundClassObject=&Fort_Ambient_Athena
 volumeMultiplier=0.752
 busSends [{audioBus=null, minSendLevel=0.6, sendLevel=1, soundSourceBus=&AmbientAudio_Exterior_SourceBus}]
 concurrencyOverrides {bLimitToOwner=true, maxCount=4, voiceStealReleaseTime=1}
 soundSubmixSends [{minSendLevel=1, soundSubmix=&Fortnite_Reverb_Default}]
}
