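# GPU Niagara system for a seagull flyby in V formation, distance-culled and using mesh and sprite renderers.
# Birds001 spawns a brief wave in two groups along mirrored offset curves, launches it along +X and repels birds from the camera.
# Speed-scaled flap counters drive dynamic material parameters.
# A position reader samples invisible, wind-driven BirdFlocker particles, but the vortex force using FishMoverPosition is disabled, as is line attraction.
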NiagaraSystem /Game/Effects/Niagara/Systems/Enviroment/Fauna/NS_Fauna_SeaGull_Flyby_VShape
dependencies {
 usedDataInterfaces [&NiagaraDataInterfaceCurve, &NiagaraDataInterfaceVectorCurve, &NiagaraDataInterfaceParticleRead]
 usedDynamicInputs []
 usedModules [&SystemState]
}
emitters {
 BirdFlocker {
  simTarget=GPUComputeSim
  emitterUpdateScript [
   &EmitterState(
    Enable Distance Culling=true
    Life Cycle Mode=::NewEnumerator1
    Loop Duration=1
    MaxDistance=10000
    Module.bApplySpawnCountScaleByDistance=true
    Module.bUseMaxDistance=true
    Module.bUseMinDistance=false
    Reset Age On Awaken=true
    Scalability Mode=::NewEnumerator1
    Scale Spawn Count=true
    Spawn Count Scale By Distance=DataInterface() {propertyValues {Curve {keys [<at 0 1>,<at 0.95 1>,<at 1 0>]}}}
    UseLoopDelay=false
   )
   &SpawnRate(SpawnRate=5, Use Spawn Probability=false)
  ]
  particleSpawnScript [
   &InitializeParticle(Lifetime=2, Sprite Size Mode=::NewEnumerator3, Uniform Sprite Size=0, UsePositionOffset=false, Write Material Random=true)
   &AddVelocity(Coordinate Space=::Simulation, Scale Added Velocity=<0 0 0>)
   &SphereLocation(
    MaskSpawnGroup=false
    Offset=&MakeVector(
     X=&FloatFromCurve(FloatCurve=DataInterface() {propertyValues {Curve {keys [<at 0 1>,<at 1 0>]}}}, CurveIndex=$System.NormalizedLoopAge, Scale Curve=250)
     Y=0
     Z=0
    )
    Override Seed=false
    Sphere Radius=500
    Surface Only=false
    UseNonUniformScale=false
    UseOffset=true
    UseSphereOrientation=false
   )
  ]
  particleUpdateScript [
   &ParticleState()
   &WindForce(Wind Speed=<1500 0 0>)
   &SolveForcesAndVelocity(Clamp Velocity=false, Limit Acceleration=false, Speed Limit=15000)
  ]
  rendererClasses [&NiagaraSpriteRendererProperties]
  renderers [{rendererClass=&NiagaraSpriteRendererProperties}]
 }
 Birds001 {
  simTarget=GPUComputeSim
  emitterUpdateScript [
   &EmitterState(
    Enable Distance Culling=true
    MaxDistance=15000
    Min Distance Response=::SleepAndLetParticlesFinish
    MinDistance=2500
    Module.bApplySpawnCountScaleByDistance=true
    Module.bUseMaxDistance=true
    Module.bUseMinDistance=true
    Reset Age On Awaken=true
    Scalability Mode=::NewEnumerator1
    Scale Spawn Count=true
    Spawn Count Scale By Distance=DataInterface() {propertyValues {Curve {keys [<at 0 0>,<at 0.1 1>,<at 0.9 1>,<at 1 0>]}}}
   )
   &SpawnRate(
    SpawnRate=&FloatFromCurve(FloatCurve=DataInterface() {propertyValues {Curve {keys [<at 0 1>,<at 0.25 0>]}}}, CurveIndex=$Emitter.Age, Scale Curve=1000)
    Use Spawn Probability=false
   )
   &SpawnBurst_Instantaneous(Spawn Count=25, Use Spawn Probability=false, UseLoopCountLimit=false) {enabled=false}
  ]
  particleSpawnScript [
   &FN_SetSpawnGroup(Number of Spawn Groups=2)
   &InitializeParticle(
    Color=<rgba 3 3 3 1>
    Lifetime=&RandomRangeFloat(Minimum=15, Maximum=18)
    Mesh Scale Mode=::NewEnumerator1
    Mesh Uniform Scale Max=1.4
    Mesh Uniform Scale Min=1.2
    Sprite Size Mode=::NewEnumerator3
    Uniform Sprite Size=500
    UsePositionOffset=false
    Write Material Random=true
   )
   &SystemLocation(
    MaskSpawnGroup=false
    Offset=&VectorFromCurve(
     VectorCurve=DataInterface() {propertyValues {XCurve {keys [<at 0 0>,<at 0.5 1>,<at 1 0>]}, YCurve {keys [<at 0 0>,<at 0.5 1>,<at 1 0>]}, ZCurve {keys [<at 0 1>,<at 0.25 1>]}}}
     CurveIndex=&ReturnNormalizedExecIndex(Normalized Index Includes One=false)
     Scale Curve=<-1500 1500 0>
    )
   ) {module=SystemLocation001}
   &SystemLocation(
    MaskSpawnGroup=true
    Offset=&VectorFromCurve(
     VectorCurve=DataInterface() {propertyValues {XCurve {keys [<at 0 0>,<at 0.5 1>,<at 1 0>]}, YCurve {keys [<at 0 0>,<at 0.5 1>,<at 1 0>]}, ZCurve {keys [<at 0 1>,<at 1 0>]}}}
     CurveIndex=&ReturnNormalizedExecIndex(Normalized Index Includes One=false)
     Scale Curve=<-1500 -1500 0>
    )
    Spawn Group Mask=1
   )
   &AddVelocity(Velocity=<500 0 0>)
   &BoxLocation(
    Box Size=<2500 5 5>
    MaskSpawnGroup=true
    Module.Override Seed=false
    Normalized Box Rotation Angle=0.1
    Rotate Box=true
    Spawn Group Mask=1
    Surface Only=false
   ) {enabled=false, module=BoxLocation001}
   &SphereLocation(
    MaskSpawnGroup=false
    Non Uniform Scale=<2 5 2>
    Override Seed=false
    Sphere Radius=25
    Surface Only=false
    UseNonUniformScale=true
    UseOffset=true
    UseSphereOrientation=false
   ) {module=SphereLocation001}
  ]
  particleUpdateScript [
   &ParticleState()
   &self:PositionReader(
    IndexID:NiagaraInt32=&RandomRangeInt(Minimum=0, Maximum=1)
    New Particle Attribute Reader=DataInterface() {propertyValues {EmitterBinding {bindingMode=Other, emitterName=BirdFlocker}}}
   ) {module=PositionReader}
   /SetVariables(Particles.RandomDirection:Vector3f=&Lerp_Vector(
     A=&MakeVector(
      X=&Waveform(
       Global Phase=&Add_Float(A=$Particles.Age, B=&RandomRangeFloat(Minimum=0, Maximum=5))
       "Waveform [1]"=::NewEnumerator8
       "[1] Frequency"=1
       "[1] Amplitude Min/Max"=<-1 1>
       "[2] Phase"=$Emitter.Age
       "[3] Phase"=$Emitter.Age
      )
      Y=&Waveform(
       Global Phase=&Add_Float(A=$Particles.Age, B=&RandomRangeFloat(Minimum=0, Maximum=5))
       "Waveform [1]"=::NewEnumerator8
       "[1] Frequency"=1
       "[1] Amplitude Min/Max"=<-1 1>
       "[2] Phase"=$Emitter.Age
       "[3] Phase"=$Emitter.Age
      )
      Z=0
     )
     B=<1 0 0>
     Alpha=&Waveform(
      Global Phase=&Add_Float(A=$Particles.Age, B=&RandomRangeFloat(Minimum=0, Maximum=5))
      Global Clamp Min=0.4
      "Waveform [1]"=::NewEnumerator8
      "[1] Frequency"=1
      "[1] Amplitude Min/Max"=<0 1>
      "[2] Phase"=$Emitter.Age
      "[3] Phase"=$Emitter.Age
     )
    )) {isSetParameters=true}
   &InterpolateOverTime(
    Target Vector Value:Vector3f=&NormalizeVector(Vector To Normalize=$Particles.RandomDirection, Safe Normalize=true, Safe Fallback Threshold=0, Safe Fallback Vector=<0 0 0.001>)
    Variable Type=::NewEnumerator2
   ) {module=InterpolateOverTime001}
   &PointAttractionForce(
    Attraction Radius=2000
    Attraction Strength=-25
    Attractor Position:Vector3f=$Emitter.CameraPosition
    Kill Within Radius=false
    Module.Use Falloff=true
    UseAttractorOffset=false
   ) {module=PointAttractionForce001}
   /Niagara/Modules/Update/Forces/LineAttractionForce(
    Attraction Strength:NiagaraFloat=&FloatFromCurve(
     FloatCurve=DataInterface() {propertyValues {Curve {keys [
         {arriveTangent=4.9254, interpMode=RCIM_Cubic, time=0, value=0.15}
         {arriveTangent=-1.3479, interpMode=RCIM_Cubic, leaveTangent=-1.3479, tangentMode=RCTM_User, time=0.5, value=1}
         {arriveTangentWeight=0.9042, interpMode=RCIM_Cubic, tangentMode=RCTM_User, time=1, value=0.3}
        ]}}}
     Scale Curve=&RandomRangeFloat(Minimum=150, Maximum=380)
    )
    Line End:Vector3f=<70000 0 0>
    Line Segment in Localspace=true
    Line Start:Vector3f=<0 0 0>
    Use Attraction Falloff=false
   ) {enabled=false}
   &VortexForce(
    Origin Pull Amount=150
    UseOriginOffset=false
    Vortex Axis=<0 1 1>
    Vortex Force Amount=500
    Vortex Origin:Vector3f=$Particles.FishMoverPosition
   ) {enabled=false}
   &SolveForcesAndVelocity(Clamp Velocity=true, Limit Acceleration=false, Speed Limit=1500)
   /SetVariables(Particles.FlapTarget:NiagaraFloat=/Niagara/DynamicInputs/Integer/V2/Round_Float(Float:NiagaraFloat=&Waveform(
      "Waveform [1]"=::NewEnumerator8
      "[1] Phase"=&Add_Float(A=$System.Age, B=&RandomRangeFloat(Minimum=0, Maximum=2))
      "[1] Frequency"=&RandomRangeFloat(Minimum=0.4, Maximum=1)
      "[1] Amplitude Min/Max"=<0 1>
      "[2] Average"=false
      "[2] Input Position"=&RandomRangeFloat(Minimum=0, Maximum=1)
      "[2] Phase"=$Particles.Age
      "[3] Phase"=$Emitter.Age
     ))) {enabled=false, isSetParameters=true}
   /SetVariables(Particles.FlapTarget=1) {isSetParameters=true}
   &InterpolateOverTime(
    Rate Of Change:NiagaraFloat=&Lerp_Float(A=0.6, B=1, Alpha=$Particles.FlapTarget)
    Target Float Value:NiagaraFloat=$Particles.FlapTarget
   )
   /SetVariables(Particles.FlapSmoothed:NiagaraFloat=$Particles.InterpolateOverTime.InterpolateOverTime_Float.Moving Average) {isSetParameters=true}
   /Niagara/Modules/Update/Utility/IncrementOverTime(Rate of Change:NiagaraFloat=&Multiply_Float(
     A=&Add_Float(A=2, B=$Particles.FlapSmoothed)
     B=&RemapRange(
      Input Value To Remap=&VectorLength(Vector=$Particles.Velocity)
      Input Min=0
      Input Max=5000
      Output Min=&RandomRangeFloat(Minimum=0.2, Maximum=0.4)
      Output Max=3.5
     )
    ))
   &DynamicMaterialParameters(
    Index 0 Param 1=&Multiply_Float(A=$Particles.FlapSmoothed, B=1)
    Index 0 Param 2=&Multiply_Float(A=$Particles.IncrementOverTime.Counter, B=1)
    Module.Param0WriteEnabled=true
    Module.Param1WriteEnabled=true
    Module.Param2WriteEnabled=false
    Module.Param3WriteEnabled=false
   ) {module=DynamicMaterialParameters001}
   &ScaleMeshSize(Scale Factor=&VectorFromFloat(Value=&FloatFromCurve(FloatCurve=DataInterface() {propertyValues {Curve {keys [
          {arriveTangent=2.8801, interpMode=RCIM_Cubic, leaveTangent=2.8801, tangentMode=RCTM_User, time=-0.0063, value=0.4982}
          {arriveTangentWeight=0.0378, interpMode=RCIM_Cubic, leaveTangentWeight=0.2841, tangentMode=RCTM_User, time=0.2, value=1}
          {arriveTangentWeight=0.4334, interpMode=RCIM_Cubic, leaveTangentWeight=0.0534, tangentMode=RCTM_User, time=0.9, value=1}
          {arriveTangent=-6.6086, interpMode=RCIM_Cubic, leaveTangent=-6.6086, tangentMode=RCTM_User, time=0.9985, value=0.5165}
         ]}}}))) {module=ScaleMeshSize001}
  ]
  rendererClasses [&NiagaraMeshRendererProperties, &NiagaraSpriteRendererProperties]
  renderers [
   {rendererClass=&NiagaraMeshRendererProperties, rendererIndex=0}
   {rendererClass=&NiagaraSpriteRendererProperties, rendererIndex=1}
  ]
 }
}
summary {
 systemName=NS_Fauna_SeaGull_Flyby_VShape
 emitters [
  {emitterName=BirdFlocker, simTarget=GPUComputeSim, rendererClasses [&NiagaraSpriteRendererProperties]}
  {
   emitterName=Birds001
   simTarget=GPUComputeSim
   rendererClasses [&NiagaraMeshRendererProperties, &NiagaraSpriteRendererProperties]
  }
 ]
 userVariables [
  User.FishOffset {type=NiagaraFloat, defaultValue=0, description=""}
  User.HideZP {type=NiagaraFloat, defaultValue=0, description=""}
  User.Push {type=NiagaraFloat, defaultValue=0, description=""}
 ]
}
