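# Niagara system for an ambient flock of circling seagulls, drawn as mesh particles.
# Only the GPU emitter BirdsInteractabler is enabled, spawning birds continuously on sphere and torus shapes sized partly by User.SpawnRadius; two orbit-initialized emitters are disabled.
# Vortex and curl noise forces move the flock, while negative point attraction at the camera position pushes nearby birds away.
# Mesh scale follows lifetime and camera-distance curves to hide popping, alongside emitter distance culling.
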NiagaraSystem /Game/Effects/Niagara/Systems/Enviroment/Fauna/NS_Fauna_Seagulls_CS_Circling
dependencies {
 usedDataInterfaces [&NiagaraDataInterfaceCamera, &NiagaraDataInterfaceCurve]
 usedDynamicInputs []
 usedModules [&SystemState]
}
emitters {
 Birds001 {
  bEnabled=false
  emitterUpdateScript [
   &EmitterScalability(
    Module.bApplySpawnCountScale=false
    Module.bApplySpawnCountScaleByDistanceFraction=false
    Module.bUseMaxDistance=false
    Module.bUseMinDistance=false
   ) {enabled=false}
   &EmitterLifeCycle() {enabled=false}
   &EmitterState(
    Enable Distance Culling=true
    Life Cycle Mode=::NewEnumerator1
    Loop Behavior=::NewEnumerator1
    Loop Duration Mode=::NewEnumerator1
    Module.bUseMaxDistance=true
    Module.bUseMinDistance=false
    Scalability Mode=::NewEnumerator1
    UseLoopDelay=false
   )
   &SpawnBurst_Instantaneous(Spawn Count=1, Use Spawn Probability=false)
  ]
  particleSpawnScript [
   /Game/Effects/Niagara/Modules/Orbit/FN_Orbit_Init_Variables_v2()
   /Niagara/Modules/Spawn/Initialization/InitializeParticle(
    Mesh Scale=&UniformRangedVector(Minimum=<1.25 1 1>, Maximum=<1.25 1 1>)
    Module.Write Color=true
    Module.Write Lifetime=true
    Module.Write Mass=true
    Module.Write Material Random=true
    Module.Write Position=false
    Module.Write Scale=true
    Module.Write SpriteRotation=false
    Module.Write SpriteSize=false
    Position=&SimulationPosition()
   )
   &AddRotationalVelocity(
    Coordinate Space:ENiagaraMeshTransforms=::NewEnumerator1
    Rotation Rate=&UniformRangedVector(Minimum=<-0.25 0 0>, Maximum=<0.25 0 0>)
   ) {enabled=false}
   &ApplyInitialForces()
   /Niagara/Modules/Spawn/Orientation/V2/InitialMeshOrientation(Use Orientation Vector=true, Use Rotation Vector=true) {enabled=false}
   &DynamicMaterialParameters(
    Index 0 Param 2=&UniformRangedFloat(Minimum=1, Maximum=1.2)
    Module.Param0WriteEnabled=true
    Module.Param1WriteEnabled=true
    Module.Param2WriteEnabled=true
    Module.Param3WriteEnabled=true
   )
  ]
  particleUpdateScript [
   &ParticleState(Kill Particles When Lifetime Has Elapsed=false, Let Infinitely Lived Particles Die When Emitter Deactivates=true)
   &Color()
   /Niagara/Modules/Update/Forces/PointAttractionForce(Attraction Radius=256, Module.Kill Within Radius=false, Module.Use Falloff=true)
   &SolveForcesAndVelocity(Clamp Velocity=false, Limit Acceleration=false)
  ]
  rendererClasses [&NiagaraMeshRendererProperties]
  renderers [{rendererClass=&NiagaraMeshRendererProperties, rendererIndex=0}]
 }
 Birds002 {
  bEnabled=false
  simTarget=GPUComputeSim
  emitterUpdateScript [
   /SetVariables(Emitter.CameraPosition:Vector3f=&ReturnCameraVector(Camera Property=::NewEnumerator0, Coordinate Space=::Simulation)) {isSetParameters=true, module=SetVariables_#1}
   &EmitterState()
   &SpawnBurst_Instantaneous(Spawn Count=250, Use Spawn Probability=false) {enabled=false}
   &SpawnRate(SpawnRate=25, Use Spawn Probability=false)
  ]
  particleSpawnScript [
   &FN_Orbit_Init_Variables_v2()
   &InitializeParticle(
    Color=<rgba 0 0 0 1>
    Lifetime=3
    Mesh Scale=&Lerp_Vector(A=<1 1 1>, B=<1 1 1>, Alpha=&UniformAOrBFloat(A=0))
    Module.Write Color=true
    Module.Write Lifetime=true
    Module.Write Mass=true
    Module.Write Material Random=true
    Module.Write Position=false
    Module.Write Scale=true
    Module.Write SpriteRotation=false
    Module.Write SpriteSize=false
    Position=&SimulationPosition()
   )
   &ApplyInitialForces()
   &SystemLocation(MaskSpawnGroup=false, Offset=&UniformRangedVector(Minimum=<-25 0 -165>, Maximum=<25 1 -185>))
   /Niagara/Modules/Spawn/Location/SphereLocation(MaskSpawnGroup=false, Module.Override Randomness=false, Non Uniform Scale=<1 1 0.2>, Sphere Radius=500, Surface Only=true)
   &DynamicMaterialParameters(
    Index 0 Param 2=&UniformRangedFloat(Minimum=1.25, Maximum=1.5)
    Module.Param0WriteEnabled=true
    Module.Param1WriteEnabled=true
    Module.Param2WriteEnabled=true
    Module.Param3WriteEnabled=true
   )
   &TorusLocation(Align Torus With Axis=false, Large Radius=2500, MaskSpawnGroup=false, Module.Override Randomness=false)
  ]
  particleUpdateScript [
   /Game/Effects/Niagara/Modules/Orbit/FN_Orbit_v4(
    Height min/max:Vector2f=<-1500 1500>
    Radius min/max:Vector2f=<1500 2000>
    Rotation Offset min/max:Vector2f=<0 1500>
    Rotation Rate min/max:Vector2f=<0.03 0.05>
    To Axis:Vector3f=&Lerp_Vector(A=<0.5 0.25 0.9>, B=<-0.5 -0.25 1>, Alpha=$Particles.Initial.Orbit.HeightSeed)
   )
   &ParticleState()
   &Color(Color=<rgba 0 0 0 1>)
   /Niagara/Modules/Update/Forces/V2/PointAttractionForce(
    Attraction Radius=1000
    Attraction Strength=-50
    Attractor Position:Vector3f=$Emitter.CameraPosition
    Kill Within Radius=false
    Module.Use Falloff=true
    UseAttractorOffset=false
   ) {module=PointAttractionForce001}
   &SolveForcesAndVelocity(Clamp Velocity=false, Limit Acceleration=false)
   &ScaleMeshSize(Scale Factor=&Multiply_Vector(
     A=<1 1 1>
     B=&VectorFromFloat(Value=&Multiply_Float(
       A=&FloatFromCurve(
        FloatCurve=DataInterface() {propertyValues {Curve {keys [
            {arriveTangent=1.7794, interpMode=RCIM_Cubic, leaveTangent=1.7794, tangentMode=RCTM_User, time=-0.0063, value=0.4982}
            {arriveTangent=0.3677, interpMode=RCIM_Cubic, leaveTangent=0.3677, tangentMode=RCTM_User, time=0.2892, value=0.9107}
            {arriveTangent=-1.2812, interpMode=RCIM_Cubic, leaveTangent=-1.2812, tangentMode=RCTM_User, time=0.8446, value=0.8923}
            <at 0.9985 0.5165>
           ]}}}
        Scale Curve=0.75
       )
       B=&FloatFromCurve(
        FloatCurve=DataInterface() {propertyValues {Curve {keys [
            <at 0.0451 -0.0521>
            {arriveTangent=3.1254, time=0.121, value=1.9893}
            {arriveTangent=-3.0163, leaveTangent=-3.0163, time=0.8976, value=2.0314}
            <at 1.0186 1.9108>
           ]}}}
        CurveIndex=&NormalizeDistanceRange(
         Start Position:NiagaraPosition=Linked() {linkedVariable {name=Emitter.CameraPosition, type=Vector3f}}
         End Position:NiagaraPosition=Linked() {linkedVariable {name=Particles.FN_Orbit_v4.Position, type=Vector3f}}
         Distance:NiagaraFloat=15000
         Invert Normalized Range:NiagaraBool=false
        )
        Scale Curve=2
       )
      ))
    ))
   &ScaleMeshSize(Scale Factor=&Multiply_Vector(
     A=<1 1 1>
     B=&VectorFromFloat(Value=&Multiply_Float(
       A=&FloatFromCurve(FloatCurve=DataInterface() {propertyValues {Curve {keys [
            {arriveTangent=1.7794, interpMode=RCIM_Cubic, leaveTangent=1.7794, tangentMode=RCTM_User, time=-0.0063, value=0.4982}
            {arriveTangent=0.3677, interpMode=RCIM_Cubic, leaveTangent=0.3677, tangentMode=RCTM_User, time=0.2892, value=0.9107}
            {arriveTangent=-1.2812, interpMode=RCIM_Cubic, leaveTangent=-1.2812, tangentMode=RCTM_User, time=0.8446, value=0.8923}
            <at 0.9985 0.5165>
           ]}}})
       B=&FloatFromCurve(
        FloatCurve=DataInterface() {propertyValues {Curve {keys [<at 0.0296 1.1447>, {interpMode=RCIM_Cubic, tangentMode=RCTM_User, time=0.3555, value=1.2901}, <at 0.8976 2.0314>, <at 1.0244 1.5622>]}}}
        CurveIndex=&NormalizeDistanceRange(
         Start Position:NiagaraPosition=Linked() {linkedVariable {name=Emitter.CameraPosition, type=Vector3f}}
         End Position:NiagaraPosition=Linked() {linkedVariable {name=Particles.FN_Orbit_v4.Position, type=Vector3f}}
         Distance:NiagaraFloat=15000
         Invert Normalized Range:NiagaraBool=false
        )
        Scale Curve=2
       )
      ))
    )) {enabled=false, module=ScaleMeshSize001}
   /Niagara/Modules/Update/Camera/MaintainInCameraParticleScale() {enabled=false}
  ]
  rendererClasses [&NiagaraMeshRendererProperties]
  renderers [{rendererClass=&NiagaraMeshRendererProperties, rendererIndex=0}]
 }
 BirdsInteractabler {
  simTarget=GPUComputeSim
  emitterUpdateScript [
   /SetVariables(Emitter.CameraPosition:Vector3f=&ReturnCameraVector(Camera Property=::NewEnumerator0, Coordinate Space=::Simulation)) {isSetParameters=true, module=SetVariables_#1}
   &EmitterState(
    Enable Distance Culling=true
    Life Cycle Mode=::NewEnumerator1
    Loop Behavior=::NewEnumerator1
    Loop Duration Mode=::NewEnumerator1
    MaxDistance=50000
    Min Distance Response=::SleepAndLetParticlesFinish
    MinDistance=2500
    Module.bApplySpawnCountScaleByDistance=true
    Module.bUseMaxDistance=true
    Module.bUseMinDistance=true
    Scalability Mode=::NewEnumerator1
    Scale Spawn Count=true
    Spawn Count Scale By Distance=DataInterface() {propertyValues {Curve {keys [<at 0 1>,<at 0.95 1>,<at 1 0>]}}}
    UseLoopDelay=false
   )
   &SpawnBurst_Instantaneous(Spawn Count=250, Use Spawn Probability=false) {enabled=false}
   &SpawnRate(
    SpawnRate=&Multiply_Float(
     A=10
     B=&FloatFromCurve(
      FloatCurve=DataInterface() {propertyValues {Curve {keys [<at 0.2027 1>,<at 0.2492 0>,<at 0.9975 0>]}}}
      CurveIndex=$Emitter.NormalizedLoopAge
      Scale Curve=&Multiply_Float(
       A=&NormalizeDistanceRange(
        Start Position:NiagaraPosition=Linked() {linkedVariable {name=Emitter.CameraPosition, type=Vector3f}}
        End Position:NiagaraPosition=$Engine.Owner.Position
        Distance:NiagaraFloat=10000
        Invert Normalized Range:NiagaraBool=false
       )
       B=1
      )
     )
    )
    Use Spawn Probability=false
   ) {enabled=false}
   &SpawnRate(SpawnRate=10, Use Spawn Probability=false) {module=SpawnRate001}
  ]
  particleSpawnScript [
   &InitializeParticle(
    Lifetime=5
    Mesh Scale=&Lerp_Vector(A=<1 1 1>, B=<1.25 1.25 1.25>, Alpha=&UniformAOrBFloat(A=0))
    Module.Write Color=true
    Module.Write Lifetime=true
    Module.Write Mass=true
    Module.Write Material Random=true
    Module.Write Position=false
    Module.Write Scale=true
    Module.Write SpriteRotation=false
    Module.Write SpriteSize=false
    Position=&SimulationPosition()
   )
   &ApplyInitialForces()
   &SphereLocation(MaskSpawnGroup=false, Module.Override Randomness=false, Non Uniform Scale=<1 1 0.4>, Sphere Radius=2000, Surface Only=true)
   &DynamicMaterialParameters(
    Index 0 Param 2=&UniformRangedFloat(Minimum=1.25, Maximum=1.5)
    Module.Param0WriteEnabled=true
    Module.Param1WriteEnabled=true
    Module.Param2WriteEnabled=true
    Module.Param3WriteEnabled=true
   )
   &TorusLocation(
    Align Torus With Axis=false
    Enable Torus Knot Radius=false
    Handle Radius=1000
    Large Radius:NiagaraFloat=$User.SpawnRadius
    MaskSpawnGroup=false
    Module.Override Randomness=false
    Torus Mode=::NewEnumerator1
   )
  ]
  particleUpdateScript [
   &ParticleState()
   &Color()
   /Niagara/Modules/Update/Forces/PointAttractionForce(Attraction Radius=256, Module.Kill Within Radius=false, Module.Use Falloff=true) {enabled=false}
   /Niagara/Modules/Update/Forces/V2/PointAttractionForce(
    Attraction Radius=2000
    Attraction Strength=-25
    Attractor Position:Vector3f=$Emitter.CameraPosition
    Kill Within Radius=false
    Module.Use Falloff=true
    UseAttractorOffset=false
   ) {module=PointAttractionForce001}
   &VortexForce(Vortex Force Amount=125)
   &CurlNoiseForce(Curl Noise Cone Mask Angle=90, Mask Curl Noise=true, Noise Frequency=15000, Noise Strength=250, Pan Noise=false)
   &SolveForcesAndVelocity(Clamp Velocity=true, Limit Acceleration=true)
   &ScaleMeshSize(Scale Factor=&Multiply_Vector(
     A=<1 1 1>
     B=&VectorFromFloat(Value=&Multiply_Float(
       A=&FloatFromCurve(FloatCurve=DataInterface() {propertyValues {Curve {keys [
            {arriveTangent=1.7794, interpMode=RCIM_Cubic, leaveTangent=1.7794, tangentMode=RCTM_User, time=-0.0063, value=0.4982}
            {arriveTangent=0.3677, interpMode=RCIM_Cubic, leaveTangent=0.3677, tangentMode=RCTM_User, time=0.2892, value=0.9107}
            {arriveTangent=-1.2812, interpMode=RCIM_Cubic, leaveTangent=-1.2812, tangentMode=RCTM_User, time=0.8446, value=0.8923}
            <at 0.9985 0.5165>
           ]}}})
       B=&FloatFromCurve(
        FloatCurve=DataInterface() {propertyValues {Curve {keys [<at 0.2476 0>, {interpMode=RCIM_Cubic, tangentMode=RCTM_User, time=0.3555, value=1.2901}, <at 0.8976 2.0314>, <at 1 0>]}}}
        CurveIndex=&NormalizeDistanceRange(
         Start Position:NiagaraPosition=Linked() {linkedVariable {name=Emitter.CameraPosition, type=Vector3f}}
         End Position:NiagaraPosition=Linked() {linkedVariable {name=Particles.FN_Orbit_v4.Position, type=Vector3f}}
         Distance:NiagaraFloat=15000
         Invert Normalized Range:NiagaraBool=false
        )
        Scale Curve=2
       )
      ))
    )) {enabled=false}
   &ScaleMeshSize(Scale Factor=&Multiply_Vector(
     A=<1 1 1>
     B=&VectorFromFloat(Value=&Multiply_Float(
       A=&FloatFromCurve(
        FloatCurve=DataInterface() {propertyValues {Curve {keys [
            {arriveTangent=1.7794, interpMode=RCIM_Cubic, leaveTangent=1.7794, tangentMode=RCTM_User, time=-0.0063, value=0.4982}
            {arriveTangent=0.3677, interpMode=RCIM_Cubic, leaveTangent=0.3677, tangentMode=RCTM_User, time=0.1297, value=0.8851}
            {arriveTangent=-1.2812, interpMode=RCIM_Cubic, leaveTangent=-1.2812, tangentMode=RCTM_User, time=0.8446, value=0.8923}
            <at 0.9985 0.5165>
           ]}}}
        Scale Curve=0.66
       )
       B=&FloatFromCurve(
        FloatCurve=DataInterface() {propertyValues {Curve {keys [<at -0.0055 -0.0388>, {interpMode=RCIM_Cubic, tangentMode=RCTM_User, time=0.0645, value=2}, <at 0.9 2>, <at 1 0>]}}}
        CurveIndex=&NormalizeDistanceRange(
         Start Position:NiagaraPosition=Linked() {linkedVariable {name=Emitter.CameraPosition, type=Vector3f}}
         End Position:NiagaraPosition=$Particles.Position
         Distance:NiagaraFloat=50000
         Invert Normalized Range:NiagaraBool=false
        )
       )
      ))
    )) {module=ScaleMeshSize001}
   &MaintainInCameraParticleScale() {enabled=false}
   &DynamicMaterialParameters(
    Index 0 Param 2=2
    Module.Param0WriteEnabled=true
    Module.Param1WriteEnabled=true
    Module.Param2WriteEnabled=true
    Module.Param3WriteEnabled=true
   ) {module=DynamicMaterialParameters001}
  ]
  rendererClasses [&NiagaraMeshRendererProperties]
  renderers [{rendererClass=&NiagaraMeshRendererProperties, rendererIndex=0}]
 }
}
summary {
 systemName=NS_Fauna_Seagulls_CS_Circling
 emitters [
  {emitterName=BirdsInteractabler, simTarget=GPUComputeSim, rendererClasses [&NiagaraMeshRendererProperties]}
  {bEnabled=false, emitterName=Birds001, rendererClasses [&NiagaraMeshRendererProperties]}
  {bEnabled=false, emitterName=Birds002, simTarget=GPUComputeSim, rendererClasses [&NiagaraMeshRendererProperties]}
 ]
 userVariables [User.SpawnRadius {type=NiagaraFloat, defaultValue=2000, description=""}]
}
